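; Level volume actor that resets the falling height of player pawns entering its box, removing the fall damage accumulated up to that point.
; On box begin overlap, authority only, it casts the other actor to FortPawn, then FortPlayerPawnAthena, and calls ResetFallingHeight.
; The construction script sizes the box from the BoxExtents variable.
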(event OnComponentBeginOverlap(Box)
  (params (OverlappedComponent :type Primitive%20Component%20Object%20Reference) (OtherActor :type Actor%20Object%20Reference) (OtherComp :type Primitive%20Component%20Object%20Reference) (OtherBodyIndex :type Integer) (bFromSweep :type Boolean) (SweepResult :type Hit%20Result%20Structure%20%28by%20ref%29))
  (Networking|SwitchHasAuthority
    (:Authority
      (bind _asfort_pawn (Utilities|Casting|CastToFortPawn OtherActor))
      (bind _asfort_player_pawn_athena (Utilities|Casting|CastToFortPlayerPawnAthena _asfort_pawn))
      (Pawn|Character|ResetFallingHeight _asfort_player_pawn_athena))))

(fn ConstructionScript
  (Components|Box|SetBoxExtent (Variables|Default|GetBox) (Variables|Default|GetBoxExtents) true))
